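# Unlit masked toon material for Niagara mesh particles, faking cel-shaded lighting.
# Sky atmosphere light direction dotted with vertex normals, blendable toward a tiled lake-water normal map, drives a contrast-adjusted ramp between dynamic-parameter shadow color and particle color.
# That result multiplies a color texture, then a channel-selectable mask blends in an outline color before shared emissive adjustments.
# Normal tiling and offset use only their X parameters for both axes.
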Material /WildEstateWeapons/Assets/FX/Materials/M_ToonParticleMesh
graph {
 properties {
  bAutomaticallySetUsageInEditor=false
  bUsedWithNiagaraMeshParticles=true
  blendMode=BLEND_Masked
  ditherOpacityMask=true
  shadingModel=MSM_Unlit
  usedShadingModels=Unlit
 }
 nodes [
  DynamicParameter() {parameterIndex=1, paramNames [NA,NA,NA,NA]}
  DynamicParameter() {parameterIndex=2, paramNames [NA,NA,NA,NA]}
  DynamicParameter() {parameterIndex=3, paramNames [NA,NA,NA,NA]}
  LinearInterpolate() {constA=-0.5, constAlpha=0, constB=0.5}
  n1 = TextureSampleParameter2D() {parameterName=Color Texture, texture=/Engine/EngineResources/WhiteSquareTexture}
  n2 = VertexNormalWS()
  n3 = VectorParameter() {parameterName=Normal Tile Offset, channelNames {a=Offset Y, b=Offset X, g=Tile Y, r=Tile X}, defaultValue=<rgba 1 1 0 0>}
  n4 = Multiply(A=TextureCoordinate(), B=AppendVector(A=@n3.Tile X, B=@n3.Tile X))
  n5 = AppendVector(A=@n3.Offset X, B=@n3.Offset X)
  n6 = Transform(None=TextureSample(UVs=Add(A=@n4, B=@n5)) {samplerType=SAMPLERTYPE_Normal, texture=/FortWater/Materials/WaterSurface/Mobile/Textures/T_MobileWater_LakeNormal}.RGB)
  n7 = ScalarParameter() {parameterName=Normal Strength}
  n8 = DotProduct(A=Normalize(VectorInput=SkyAtmosphereLightDirection()), B=Normalize(VectorInput=LinearInterpolate(A=@n2, B=@n6, Alpha=@n7)))
  n9 = VectorParameter() {parameterName=ShadowLine Options, channelNames {a=contrast, b=Offset, g=mul, r=pow}, defaultValue=<rgba 1 1 0 100>}
  n10 = Multiply(A=Power(Base=ConstantBiasScale(None=@n8), Exp=@n9.pow), B=@n9.mul)
  n11 = LinearInterpolate(
   A=DynamicParameter() {paramNames [Shadow Color R, Shadow Color G, Shadow Color B, NA]}.RGB
   B=ParticleColor().RGB
   Alpha=MaterialFunctionCall(In=Add(A=@n10, B=@n9.Offset), Contrast=@n9.contrast) {materialFunction=&CheapContrast}.Result
  )
  n12 = ChannelMaskParameter(None=TextureSampleParameter2D() {parameterName="Outline Mask (R)", texture=&Black}.RGBA) {parameterName=OutlinMaskChannel}
  n13 = VectorParameter() {parameterName=Outline Mask Options, channelNames {a=contrast, b=Offset, g=mul, r=pow}, defaultValue=<rgba 1 1 0 100>}
  n14 = Add(A=Multiply(A=Power(Base=@n12.R, Exp=@n13.pow), B=@n13.mul), B=@n13.Offset)
  MaterialFunctionCall(Emissive=LinearInterpolate(
    A=Multiply(A=@n1.RGB, B=@n11)
    B=VectorParameter() {parameterName=Outline Color}.RGB
    Alpha=MaterialFunctionCall(In=@n14, Contrast=@n13.contrast) {materialFunction=&CheapContrast}.Result
   )) {materialFunction=&MF_FX_EmissiveAdjustments}
  MaterialFunctionCall() {materialFunction=&DebugTimeSine}
  Multiply()
  PixelNormalWS()
  Subtract(A=@n9.Offset) {constB=0.5}
 ]
}
